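# Physics asset for the ScareyBeary backpack cosmetic, giving capsule bodies and joints to its two ribbons and two wings for secondary motion.
# Ribbons are three-segment chains and wings single bodies, each hung through a small master body from the two-capsule pack, which joins to root.
# Pack and ribbon master joints leave swing and twist free, while wing and ribbon segment joints use asymmetric swing limits with locked or narrow twist.
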PhysicsAsset /BRCosmetics/Accessories/FORT_Backpacks/Backpack_F_MED_ScareyBeary/Meshes/F_MED_ScareyBeary_Pack_Physics1
bodies {
 dyn_mstr_ribbon_l [{shapeType=Capsule, length=0.505, radius=0.5, shapeName=dyn_mstr_ribbon_l_capsule, rotation=<rot 0 -90 0>}]
 dyn_mstr_ribbon_r [{shapeType=Capsule, length=0.505, radius=0.505, shapeName=dyn_mstr_ribbon_r_capsule, rotation=<rot 0 -90 0>}]
 dyn_mstr_wing_l [{shapeType=Capsule, length=1.5, radius=0.5, shapeName=dyn_mstr_wing_l_capsule}]
 dyn_mstr_wing_r [{shapeType=Capsule, length=1.5, radius=0.5, shapeName=dyn_mstr_wing_r_capsule}]
 dyn_ribbon_l_01 [{shapeType=Capsule, length=5.5, radius=2, shapeName=dyn_ribbon_l_01_capsule, center=<4.3715 0.0926 0>, rotation=<rot 90 0 0>}]
 dyn_ribbon_l_02 [{shapeType=Capsule, length=6, radius=1.6, shapeName=dyn_ribbon_l_02_capsule, center=<4.2144 -0.1918 0>, rotation=<rot 90 0 0>}]
 dyn_ribbon_l_03 [{shapeType=Capsule, length=4, radius=1.7, shapeName=dyn_ribbon_l_03_capsule, center=<3.2147 2.3123 0>, rotation=<rot 90 180 144.2727>}]
 dyn_ribbon_r_01 [{shapeType=Capsule, length=3, radius=1.9, shapeName=dyn_ribbon_r_01_capsule, center=<2.7649 0 0>, rotation=<rot 90 0 0>}]
 dyn_ribbon_r_02 [{shapeType=Capsule, length=5, radius=1.8, shapeName=dyn_ribbon_r_02_capsule, center=<3.9433 0 0>, rotation=<rot 90 0 0>}]
 dyn_ribbon_r_03 [{
   shapeType=Capsule
   length=3.5
   radius=1.8
   shapeName=dyn_ribbon_r_03_capsule
   center=<2.4239 2.0835 -0.1103>
   rotation=<rot 86.3559 97.8557 57.1217>
  }]
 dyn_wing_l [{shapeType=Capsule, length=9.2491, radius=4.7, shapeName=dyn_wing_l_capsule, center=<-1.8283 -1.7 -7.8317>, rotation=<rot 0 85.5774 -67.2626>}]
 dyn_wing_r [{shapeType=Capsule, length=9.2765, radius=4.5, shapeName=dyn_wing_r_capsule, center=<1.2051 2 6.9969>, rotation=<rot 0 91.102 -76.3092>}]
 pack [
  {shapeType=Capsule, length=30, radius=16, shapeName=pack_capsule, center=<-0.0754 16.3177 -11.1355>}
  {shapeType=Capsule, length=15, radius=15, shapeName=pack_capsule, center=<0 15.6219 0.0001>, rotation=<rot 88.5055 -179.9927 -179.9926>}
 ]
 root [{shapeType=Capsule, length=0.505, radius=0.505, shapeName=root_capsule, rotation=<rot 0 -90 0>}]
}
constraints [
 {
  bone1Name=dyn_mstr_ribbon_l
  bone2Name=pack
  swing1LimitDegrees=45
  swing1Motion=Free
  swing2LimitDegrees=45
  swing2Motion=Free
  twistLimitDegrees=45
  twistMotion=Free
 }
 {bone1Name=dyn_ribbon_l_01, bone2Name=dyn_mstr_ribbon_l, swing1LimitDegrees=25, swing2LimitDegrees=4, twistLimitDegrees=1, twistMotion=Locked}
 {bone1Name=dyn_ribbon_l_02, bone2Name=dyn_ribbon_l_01, swing1LimitDegrees=20, swing2LimitDegrees=8, twistLimitDegrees=2}
 {bone1Name=dyn_ribbon_l_03, bone2Name=dyn_ribbon_l_02, swing1LimitDegrees=28, swing2LimitDegrees=10, twistLimitDegrees=1, twistMotion=Locked}
 {
  bone1Name=dyn_mstr_ribbon_r
  bone2Name=pack
  swing1LimitDegrees=45
  swing1Motion=Free
  swing2LimitDegrees=45
  swing2Motion=Free
  twistLimitDegrees=45
  twistMotion=Free
 }
 {bone1Name=dyn_ribbon_r_01, bone2Name=dyn_mstr_ribbon_r, swing1LimitDegrees=21, swing2LimitDegrees=4, twistLimitDegrees=1, twistMotion=Locked}
 {bone1Name=dyn_ribbon_r_02, bone2Name=dyn_ribbon_r_01, swing1LimitDegrees=27, swing2LimitDegrees=10, twistLimitDegrees=3}
 {bone1Name=dyn_ribbon_r_03, bone2Name=dyn_ribbon_r_02, swing1LimitDegrees=32, swing2LimitDegrees=17, twistLimitDegrees=3}
 {bone1Name=dyn_mstr_wing_l, bone2Name=pack, swing1LimitDegrees=5, swing2LimitDegrees=40, twistLimitDegrees=1, twistMotion=Locked}
 {bone1Name=dyn_mstr_wing_r, bone2Name=pack, swing1LimitDegrees=5, swing2LimitDegrees=40, twistLimitDegrees=1, twistMotion=Locked}
 {bone1Name=dyn_wing_l, bone2Name=dyn_mstr_wing_l, swing1LimitDegrees=2, swing2LimitDegrees=20, twistLimitDegrees=1, twistMotion=Locked}
 {bone1Name=dyn_wing_r, bone2Name=dyn_mstr_wing_r, swing1LimitDegrees=2, swing2LimitDegrees=25, twistLimitDegrees=1, twistMotion=Locked}
 {
  bone1Name=pack
  bone2Name=root
  swing1LimitDegrees=45
  swing1Motion=Free
  swing2LimitDegrees=45
  swing2Motion=Free
  twistLimitDegrees=45
  twistMotion=Free
 }
]
