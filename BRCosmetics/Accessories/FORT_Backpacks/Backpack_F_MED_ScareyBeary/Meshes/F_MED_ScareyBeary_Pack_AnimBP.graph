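# Anim Blueprint for the ScareyBeary backpack that animates its dynamic ribbon bones with physics and procedural offsets.
# A RigidBody node simulates in the pack bone's space with custom gravity, then additive bone modifiers offset both ribbons while skydiving.
# A final half-alpha modifier further shifts and rolls the left ribbon.
# Each update sets the skydiving alpha and maps the owner mesh's spine yaw, relative to ref pose, into that left-ribbon offset.
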AnimBlueprint /BRCosmetics/Accessories/FORT_Backpacks/Backpack_F_MED_ScareyBeary/Meshes/F_MED_ScareyBeary_Pack_AnimBP
target_skeleton=&/F_MED_ScareyBeary_Pack_Skeleton
graphs [
 AnimGraph {nodes [
   n1 = RigidBody(ComponentPose=Local To Component(LocalPose=Input Pose())) {
    bOverrideWorldGravity=true
    overridePhysicsAsset=&/F_MED_ScareyBeary_Pack_Physics1
    simulationSpace=BaseBoneSpace
    alphaBoolBlend {blendOption=Linear}
    baseBoneRef {boneName=pack}
    componentLinearAccScale=<1 0.3 0.78>
    componentLinearVelScale=<1 0.3 0.78>
    overrideWorldGravity {z=-980}
   }
   AnimGraphNode_ModifyBone_2 = "Transform (Modify) Bone - Bone: dyn_mstr_ribbon_l"(ComponentPose=@n1, Alpha=Get skydiving() -> (skydiving:double), Translation="0,5,0", Rotation="0,0,25", Scale="1,1,1") {
    binding=&self:AnimGraph.AnimGraphNode_ModifyBone_2.Binding
    rotationMode=BMM_Additive
    scaleMode=BMM_Additive
    translationMode=BMM_Additive
    alphaBoolBlend {blendOption=Linear}
    boneToModify {boneName=dyn_mstr_ribbon_l}
    rotation {roll=25}
    translation {y=5}
   }
   AnimGraphNode_ModifyBone_3 = "Transform (Modify) Bone - Bone: dyn_mstr_ribbon_r"(ComponentPose=@AnimGraphNode_ModifyBone_2, Alpha=Get skydiving(), Translation="0,0,0", Rotation="0,0,25", Scale="1,1,1") {
    binding=&self:AnimGraph.AnimGraphNode_ModifyBone_3.Binding
    rotationMode=BMM_Additive
    scaleMode=BMM_Additive
    translationMode=BMM_Additive
    alphaBoolBlend {blendOption=Linear}
    boneToModify {boneName=dyn_mstr_ribbon_r}
    rotation {roll=25}
   }
   AnimGraphNode_ModifyBone_1 = "Transform (Modify) Bone - Bone: dyn_mstr_ribbon_l"(
    ComponentPose=@AnimGraphNode_ModifyBone_3
    Alpha=0.5
    Translation=Get backpackTranslation() -> (backpackTranslation:Vector)
    Rotation=Get backpackRotate() -> (backpackRotate:Rotator)
    Scale="1,1,1"
   ) {
    binding=&self:AnimGraph.AnimGraphNode_ModifyBone_1.Binding
    rotationMode=BMM_Additive
    scaleMode=BMM_Additive
    translationMode=BMM_Additive
    alphaBoolBlend {blendOption=Linear}
    boneToModify {boneName=dyn_mstr_ribbon_l}
    rotation {roll=12}
   }
   Output Pose(Result=Component To Local(ComponentPose=@AnimGraphNode_ModifyBone_1)) {layerGroup=None}
  ]}
 EventGraph {nodes [
   Event BlueprintUpdateAnimation()
   > n1 = Sequence!(execute) -> (then_0, then_1, then_2:Exec)
   > n2 = Branch(Condition=Get bIsSkyDiving())
   > Set skydiving(skydiving:double=1) -> (Output_Get:double)
   Set skydiving(execute=@n2.else, skydiving=0)
   n3 = GetDeltaTransformFromRefPose!(self=GetOwnerSkeletalMeshComponent(), BoneName=spine_05, BaseName=spine_01) -> (
    ReturnValue_Location:Vector
    ReturnValue_Rotation_Roll:float
    ReturnValue_Rotation_Pitch:float
    ReturnValue_Rotation_Yaw:float
    ReturnValue_Scale:Vector
   )
   Set backpackRotate(
    execute=@n1.then_1
    backpackRotate:Rotator=MakeRotator(Roll=MapRangeClamped(Value=@n3.ReturnValue_Rotation_Yaw, InRangeA=0, InRangeB=-30, OutRangeB=35))
   ) -> (Output_Get:Rotator)
   > Set backpackTranslation(backpackTranslation:Vector=MakeVector(Y=MapRangeClamped(Value=@n3.ReturnValue_Rotation_Yaw, InRangeA=0, InRangeB=-30, OutRangeB=-10))) -> (Output_Get:Vector)
  ]}
]
variables [
 backpackRotate {type=Rotator}
 backpackTranslation {type=Vector}
 skydiving {type=number}
]
